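# Two-sided Single Layer Water material for Styx streams, usable with Nanite.
# World-space detail noise is flow-mapped by vertex color or mesh tangent, then its flattened normal and sine-wave offset feed the water simulation texture.
# The resulting attributes and scene depth enter MF_Water_Styx, whose scattering and absorption are masked by face side and view direction.
# A depth-fade emissive chain exists but does not feed the Emissive Color input.
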Material /Game/Environments/Helios/TechArt/Styx/Streams/Material/M_StyxStreams
graph {
 properties {
  bUseMaterialAttributes=true
  bUsedWithNanite=true
  physMaterial=/FortWater/Materials/WaterSurface/WaterBody
  pixelDepthOffsetMode=PDOM_Legacy
  shadingModel=MSM_SingleLayerWater
  twoSided=true
  usedShadingModels=SingleLayerWater
 }
 nodes [
  n1 = ScalarParameter() {defaultValue=0.272, parameterName=Dist}
  AppendVector(A=@n1, B=@n1)
  Constant3Vector() {constant=<rgba 0 0 1 0>}
  Constant() {r=100}
  CrossProduct(A=Constant3Vector() {constant=<rgba 0 0 1 0>}, B=Transform(None=Constant3Vector() {constant=<rgba 1 0 0 0>}))
  n2 = Min(A=Subtract(A=SceneDepthWithoutWater!(UVs).R, B=PixelDepth().R), B=ScalarParameter() {defaultValue=50, parameterName=MaxDepth})
  n3 = DotProduct(A=CameraVectorWS(), B=VertexNormalWS())
  Divide(A=@n2, B=@n3)
  n4 = Multiply(A=Time(), B=ScalarParameter() {defaultValue=0.456, parameterName=WaveSpeed})
  n5 = WorldPosition()
  n6 = ScalarParameter() {defaultValue=361.4055, parameterName=WaveFrequency}
  n7 = MaterialFunctionCall(Float2=Add(A=Frac(None=@n4), B=Divide(A=@n5.XY, B=@n6))) {materialFunction=&BreakOutFloat2Components}
  n8 = Add(A=Multiply(A=Sine(None=@n7.R), B=Sine(None=@n7.G)), B=ScalarParameter() {defaultValue=-0.5, parameterName=WPO_WaveHeight_Bias})
  n9 = ScalarParameter() {defaultValue=2, parameterName=WPO_WaveHeight}
  n10 = VertexColor()
  n11 = AppendVector(A=@n10.R, B=@n10.R)
  n12 = Multiply(A=Constant3Vector() {constant=<rgba 1 1 0 0>}, B=VertexTangentWS())
  n13 = ComponentMask(None=StaticSwitchParameter(True=Subtract(A=@n11) {constB=0.5}, False=Normalize(VectorInput=@n12)) {parameterName=UseVertexColorForFlowMap}) {g=true, r=true}
  n14 = ScalarParameter() {defaultValue=0.24, parameterName=OffsetDistance}
  n15 = MaterialFunctionCall!(
   Diffuse=TextureObject() {samplerType=SAMPLERTYPE_Masks, texture=/FortWater/Textures/Styx/StyxWater_DetailNoise_M}
   Normal Map=TextureObject() {samplerType=SAMPLERTYPE_Normal, texture=/FortWater/Textures/Styx/StyxWater_DetailNoise_N}
   "Flow Vector Map (see tooltip)"=Multiply(A=VertexColor().B, B=Multiply(A=@n13, B=@n14))
   UV texture 2 offset=ScalarParameter() {defaultValue=0.5, parameterName=OffsetTexture}
   "Time (see tooltip)"=Multiply(A=Time(), B=ScalarParameter() {defaultValue=0.25, parameterName=Speed})
   useMipLevel
   UVs=Divide(A=WorldPosition().XY, B=ScalarParameter() {defaultValue=722.8368, parameterName=Water_Size})
   Texture Size
   Texture Mip Bias
  ) {materialFunction=&/FlowMaps_TangentSpace}
  n16 = MaterialFunctionCall!(
   Flowmap Velocity
   Previous Foam
   Normal=MaterialFunctionCall(
    Normal=Transform(None=@n15.Normal) {transformSourceType=TRANSFORMSOURCE_World, transformType=TRANSFORM_Tangent}
    Flatness=ScalarParameter() {defaultValue=0.8, parameterName=FlattenNormal}
   ) {materialFunction=&FlattenNormal}.Result
   WPO=AppendVector(A=Constant2Vector(), B=Multiply(A=@n8, B=@n9))
  ) {materialFunction=/FortWater/Materials/Functions/WaterSimulationTexture}
  n17 = SetMaterialAttributes!(
   MaterialAttributes
   Base Color=VectorParameter() {parameterName=BaseColor, defaultValue=<rgba 0 0.1 0.1 0.1>}.RGB
   Metallic
   Specular=ScalarParameter() {defaultValue=0.3, parameterName=Specular}
   Roughness=ScalarParameter() {defaultValue=0.1, parameterName=Roughness}
   Anisotropy
   Emissive Color
   Opacity=ScalarParameter() {parameterName=Opacity}
   Opacity Mask=MaterialFunctionCall(Alpha Threshold=VertexColor().R) {materialFunction=&DitherTemporalAA}.Result
   Normal=@n16.Normal
   World Position Offset=@n16.Normal
  ) {attributeSetTypes [
    #1
    #2
    #3
    #4
    #5
    #6
    #7
    #8
    #9
    #10
   ]}
  n18 = Divide(
   A=Subtract(A=SceneDepthWithoutWater().R, B=PixelDepth().R)
   B=ScalarParameter() {defaultValue=8.9982, group=Styx, parameterName=Styx_DepthFromBehindwater}
  )
  n19 = MaterialFunctionCall!(
   DefaultMaterial=@n17
   Material_NoFoam=@n17
   Depth=@n18
   SimWaterNormal
   SimErasure=@n16.Sim Foam Erasure
   TransitionMask
   Scattering
   Absorption
   Is Translucent Mat
   Mobile Reflections
  ) {materialFunction=/FortWater/Materials/Functions/MF_Water_Styx}
  MaterialFunctionCall!(In=@n19.Result, TransitionMask=@n19.Result, Is Translucent Mat) {materialFunction=/FortWater/Materials/Functions/MF_Water_Styx_SpecularOverride}
  n20 = VectorParameter() {parameterName=Absorption, defaultValue=<rgba 0.4907 1 0.1808 29.6154>}
  Multiply(A=@n20.RGB, B=@n20.RGB)
  n21 = VectorParameter() {parameterName=Scatter, defaultValue=<rgba 0.2396 1 0.479 0.2758>}
  Multiply(A=@n21.RGB, B=Divide(A=@n21.A) {constB=100})
  n22 = ComponentMask(None=@n15.Diffuse) {g=true}
  n23 = ScalarParameter() {defaultValue=2.8558, parameterName=EmissiveHeightmapContrast}
  n24 = Divide(
   A=NamedRerouteDeclaration(None=Multiply(A=@n2, B=@n3)) {name=Depth, nodeColor=<rgba 0.2891 1 0 1>}
   B=LinearInterpolate(
    A=ScalarParameter() {defaultValue=17.8378, group=Styx, parameterName=Styx_DepthFade_low}
    B=ScalarParameter() {defaultValue=493.9169, group=Styx, parameterName=Styx_DepthFade_High}
    Alpha=Power(Base=@n22, Exp=@n23)
   )
  )
  n25 = ScalarParameter() {defaultValue=1.2043, group=Styx, parameterName=Styx_DepthFade_Falloff}
  n26 = LinearInterpolate(
   B=VectorParameter() {parameterName=EmissiveColor, defaultValue=<rgba 0.2656 1 0.6019 1>}.RGB
   Alpha=Saturate(None=Power(Base=@n24, Exp=@n25))
  )
  n27 = Multiply(
   A=VectorParameter() {parameterName=EmissiveColor_Detail, defaultValue=<rgba 0.0483 2 1.4759 1>}.RGB
   B=Power(Base=ComponentMask(None=@n15.Diffuse) {r=true}, Exp=ScalarParameter() {defaultValue=3.1589, parameterName=EmissiveDetailContrast})
  )
  Reroute(None=Divide(A=Add(A=@n26, B=@n27), B=EyeAdaptation().EyeAdaptation))
  Saturate(None=@n18)
  Sine(None=@n4)
  n28 = ComponentMask(None=CameraVectorWS()) {b=true}
  n29 = LinearInterpolate(A=Saturate(None=Ceil(None=@n28)), Alpha=Saturate(None=TwoSidedSign()))
  SingleLayerWaterMaterialOutput!(
   ScatteringCoefficients=Multiply(A=@n19.Scatter, B=@n29)
   AbsorptionCoefficients=Multiply(A=Divide(B=@n19.Absorbtion), B=@n29)
   PhaseG=Constant() {r=0.5}
   ColorScaleBehindWater
  )
  n30 = Frac(None=Time())
  n31 = Multiply(B=@n1)
  TextureSample(UVs=Add(A=Multiply(A=@n30, B=@n31))) {samplerType=SAMPLERTYPE_Masks, texture=&StyxWater_DetailNoise_M}
 ]
}
